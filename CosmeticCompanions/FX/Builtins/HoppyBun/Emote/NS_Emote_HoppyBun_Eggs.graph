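# Niagara effect for the HoppyBun companion emote that launches colored eggs which burst into confetti and glitter.
# Eggs002 shoots egg sprites from six cycled skeletal mesh bone/socket indices, while CoreFlash004 adds brief mesh, sprite and light flashes at random ones.
# Confetti and glitter emitters use particle reads to spawn from eggs reaching burst age.
# Shared palette and time-of-day scaling drive color; CrowdCheerCurve strengthens confetti curl noise.
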NiagaraSystem /CosmeticCompanions/FX/Builtins/HoppyBun/Emote/NS_Emote_HoppyBun_Eggs
dependencies {
 usedDataInterfaces [
  &NiagaraDataInterfaceCurve
  &NiagaraDataInterfaceParticleRead
  &NiagaraDataInterfaceColorCurve
  &NiagaraDataInterfaceVectorCurve
  &NiagaraDataInterfaceSkeletalMesh
  &NiagaraDataInterfaceArrayColor
  &NiagaraDataInterfaceArrayFloat3
 ]
 usedDynamicInputs [&SelectVectorFromArray, &AddVectorToPosition]
 usedModules [
  &self:SystemSpawnScript.SystemScriptSource.SystemScriptGraph.NiagaraNodeAssignment_0.SetVariables_#1
  &self:SystemSpawnScript.SystemScriptSource.SystemScriptGraph.NiagaraNodeAssignment_1.SetVariables_#2
  &self:SystemSpawnScript.SystemScriptSource.SystemScriptGraph.NiagaraNodeAssignment_3.SetVariables_#3
  &self:SystemSpawnScript.SystemScriptSource.SystemScriptGraph.NiagaraNodeAssignment_20.SetVariables_#4
  &self:SystemSpawnScript.SystemScriptSource.SystemScriptGraph.NiagaraNodeAssignment_24.SetVariables_#5
  &self:SystemSpawnScript.SystemScriptSource.SystemScriptGraph.NiagaraNodeAssignment_25.SetVariables_#6
  &self:SystemSpawnScript.SystemScriptSource.SystemScriptGraph.NiagaraNodeAssignment_26.SetVariables_#7
  &self:SystemSpawnScript.SystemScriptSource.SystemScriptGraph.NiagaraNodeAssignment_28.SetVariables_#8
  &self:SystemSpawnScript.SystemScriptSource.SystemScriptGraph.NiagaraNodeAssignment_29.SetVariables_#9
  &SystemState
  &self:SystemSpawnScript.SystemScriptSource.SystemScriptGraph.NiagaraNodeAssignment_2.SetVariables_#10
 ]
}
emitters {
 Confetti001 {
  emitterUpdateScript [
   &EmitterState()
   &SpawnParticlesFromOtherEmitter(
    Attribute Reader=DataInterface() {propertyValues {EmitterBinding {bindingMode=Other, emitterName=Eggs002}}}
    Calculate Spawn Rate Per Particle=false
    Minimum Age Spawn Threshold:NiagaraFloat=&Add_Float(A=$System.Egg Burst Time, B=0)
    Spawn Rate=15
    Use Maximum Age=false
    Use Minimum Age=true
   ) {module=SpawnParticlesFromOtherEmitter001}
  ]
  particleSpawnScript [
   &InitializeParticle(
    Color=&SelectLinearColorFromArray(
     Array Sampling Mode=::NewEnumerator1
     Direct Array Index=&RandomRangeInt(Minimum=0, Maximum=7)
     Color Selection Array=$System.EggColors
     Random Seed=?
    )
    Lifetime Max=1.7
    Lifetime Mode=::NewEnumerator1
    Position Offset:NiagaraPosition=$System.Egg 01 Position
    Sprite Size Mode=::NewEnumerator1
    Uniform Sprite Size Max=6
    Uniform Sprite Size Min=3
    UsePositionOffset=true
    Write Material Random=true
   )
   &SampleParticlesFromOtherEmitter(
    Applied Position Offset:Vector3f=$System.Egg 01 Position
    Applied Velocity Scale:NiagaraFloat=&RandomRangeFloat(Minimum=0.9, Maximum=1.8)
    Reject Particle If Sampled Index Is Invalid=false
    Velocity Sampling=::NewEnumerator0
   )
   &ShapeLocation(
    Box Size=<5 5 5>
    MaskSpawnGroup=true
    Non Uniform Scale=<1 1 0.5>
    Offset Mode=::NewEnumerator0
    Override Seed=false
    Shape Primitive=::NewEnumerator2
    Spawn Group Mask=1
    Surface Only=false
    UseOwnerScale=true
   ) {module=ShapeLocation001}
   &Color(Color=&MultiplyLinearColor(
     Linear Color=$Particles.Color
     Scale Factor=&MakeVector4_FromLinearColor(LinearColor=&FNTimeOfDayColor(
       Night Color=<rgba 0.3782 0.3614 0.4948 1>
       Morning Color=<rgba 1 0.7262 0.6963 1>
       Day Color=<rgba 1 0.9315 0.9012 1>
       Evening Color=<rgba 1 0.9069 0.7687 1>
      ))
    ))
   &AddVelocity(
    Module.Use Falloff Distance=false
    Origin Offset:NiagaraPosition=$Particles.ShapeLocation.ShapePosition
    Override Seed=false
    Velocity Mode=::NewEnumerator1
    Velocity Speed=&MaskFloatBySpawnGroup(Masked Float=0, Passthrough Float=&RandomRangeFloat(Minimum=80, Maximum=145))
   )
  ]
  particleUpdateScript [
   &Drag(
    Drag=&MaskFloatBySpawnGroup(
     Masked Float=5.5
     Passthrough Float=&FloatFromCurve(
      FloatCurve=DataInterface() {propertyValues {Curve {keys [<at 0 0 RCIM_Cubic>, {arriveTangent=31.2634, interpMode=RCIM_Cubic, leaveTangent=31.2634, tangentMode=RCTM_User, time=0.1, value=1}]}}}
      Scale Curve=&RandomRangeFloat(Minimum=5.5, Maximum=8.5)
     )
    )
    Use Linear Drag=true
    Use Rotational Drag=true
   )
   &GravityForce(Gravity=&RandomRangeVector(Minimum=<0 0 -110>, Maximum=<0 0 -400>))
   &CurlNoiseForce(
    Bias Noise=false
    Noise Frequency=10
    Noise Strength=&Multiply_Float(
     A=&RandomRangeFloat(Minimum=30, Maximum=50)
     B=&Lerp_Float(A=3, B=10, Alpha=$User.CrowdCheerCurve)
    )
    Pan Noise=true
    Pan Noise Field=<0.2 0.2 0.2>
   )
   &Collision(
    Age Colliding Particles=false
    KillOnCollision=false
    Module.Control Rest State Via Surface Normal=false
    Module.EnableMaxCollisionCount=false
    Module.Randomize Collision Normal Vector=false
    Restitution=0
   )
   &SolveForcesAndVelocity(Clamp Velocity=false, Limit Acceleration=false)
   &ParticleState()
   &ScaleSpriteSize(Uniform Curve Sprite Scale=DataInterface() {propertyValues {Curve {keys [
        {interpMode=RCIM_Cubic, leaveTangent=12.0846, tangentMode=RCTM_Break, time=0, value=0}
        {interpMode=RCIM_Cubic, tangentMode=RCTM_Break, time=0.1, value=1}
        {arriveTangent=-2.374, tangentMode=RCTM_User, time=1, value=0}
       ]}}})
   /SetVariables(Particles.SubImageIndex=&FloatFromCurve(
     FloatCurve=DataInterface() {propertyValues {Curve {keys [<at 0 0>,<at 1 1>]}}}
     Scale Curve=&UniformAOrBFloat(A=350, B=450, Distribution Weight=0.1, Evaluation Type=::NewEnumerator0, Recalculate Random Each Loop=true, Override Seed=false)
    )) {isSetParameters=true}
   &ScaleColor(
    Scale Mode=1
    Scale RGBA=&MakeVector4_FromLinearColor(LinearColor=&FNTimeOfDayColor(
      Night Color=&ColorFromCurve(ColorCurve=DataInterface() {propertyValues {
         AlphaCurve {keys [<at 0 1 RCIM_Cubic>]}
         BlueCurve {keys [<at 0 1 RCIM_Cubic>,<at 0.4559 1>,<at 0.4752 1000>,<at 0.4979 1>,<at 1 1 RCIM_Cubic>]}
         GreenCurve {keys [<at 0 1 RCIM_Cubic>,<at 0.4559 1>,<at 0.4752 1000>,<at 0.4979 1>,<at 1 1 RCIM_Cubic>]}
         RedCurve {keys [<at 0 1 RCIM_Cubic>,<at 0.4559 1>,<at 0.4752 1000>,<at 0.4979 1>,<at 1 1 RCIM_Cubic>]}
        }})
      Morning Color=&ColorFromCurve(ColorCurve=DataInterface() {propertyValues {
         AlphaCurve {keys [<at 0 1 RCIM_Cubic>]}
         BlueCurve {keys [<at 0 1 RCIM_Cubic>,<at 0.4559 1>,<at 0.4752 1000>,<at 0.4979 1>,<at 1 1 RCIM_Cubic>]}
         GreenCurve {keys [<at 0 1 RCIM_Cubic>,<at 0.4559 1>,<at 0.4752 1000>,<at 0.4979 1>,<at 1 1 RCIM_Cubic>]}
         RedCurve {keys [<at 0 1 RCIM_Cubic>,<at 0.4559 1>,<at 0.4752 1000>,<at 0.4979 1>,<at 1 1 RCIM_Cubic>]}
        }})
      Day Color=&ColorFromCurve(ColorCurve=DataInterface() {propertyValues {
         AlphaCurve {keys [<at 0 1 RCIM_Cubic>]}
         BlueCurve {keys [<at 0 1 RCIM_Cubic>,<at 0.4559 1>,<at 0.4752 500>,<at 0.4979 1>,<at 1 1 RCIM_Cubic>]}
         GreenCurve {keys [<at 0 1 RCIM_Cubic>,<at 0.4559 1>,<at 0.4752 500>,<at 0.4979 1>,<at 1 1 RCIM_Cubic>]}
         RedCurve {keys [<at 0 1 RCIM_Cubic>,<at 0.4559 1>,<at 0.4752 500>,<at 0.4979 1>,<at 1 1 RCIM_Cubic>]}
        }})
      Evening Color=&ColorFromCurve(ColorCurve=DataInterface() {propertyValues {
         AlphaCurve {keys [<at 0 1 RCIM_Cubic>]}
         BlueCurve {keys [<at 0 1 RCIM_Cubic>,<at 0.4559 1>,<at 0.4752 1000>,<at 0.4979 1>,<at 1 1 RCIM_Cubic>]}
         GreenCurve {keys [<at 0 1 RCIM_Cubic>,<at 0.4559 1>,<at 0.4752 1000>,<at 0.4979 1>,<at 1 1 RCIM_Cubic>]}
         RedCurve {keys [<at 0 1 RCIM_Cubic>,<at 0.4559 1>,<at 0.4752 1000>,<at 0.4979 1>,<at 1 1 RCIM_Cubic>]}
        }})
     ))
    ScaleRGBA=true
   )
  ]
  rendererClasses [&NiagaraSpriteRendererProperties]
  renderers [{rendererClass=&NiagaraSpriteRendererProperties}]
 }
 CoreFlash004 {
  emitterUpdateScript [
   &EmitterState(Life Cycle Mode=::NewEnumerator1, Loop Behavior=::NewEnumerator1, Loop Duration=1.1, UseLoopDelay=false)
   &SpawnRate(SpawnRate=7, Use Spawn Probability=false)
  ]
  particleSpawnScript [
   &InitializeParticle(
    Lifetime=0.1
    Mesh Scale Max=<0.35 0.2 0.2>
    Mesh Scale Min=<0.2 0.15 0.15>
    Mesh Scale Mode=::NewEnumerator2
    Sprite Rotation Mode=::NewEnumerator2
    Sprite Size=&RandomRangeVector2D(Minimum=<7 8>, Maximum=<11 12>)
    Sprite Size Mode=::NewEnumerator4
    UsePositionOffset=false
    Write Material Random=true
   ) {module=InitializeParticle001}
   /SetVariables(Particles.LightPositionNew:Vector3f=&Add_Vector(
     A=Linked() {linkedVariable {name=Particles.Position, type=NiagaraPosition}}
     B=/Niagara/DynamicInputs/Transforms/TransformPosition(Position=<25 0 0>, Source Space=::Local, Destination Space=::Simulation)
    )) {isSetParameters=true}
   &InitialMeshOrientation(Mesh Orientation Mode=::NewEnumerator2, Orientation Coordinate Space=::Local, Orientation Vector=<0 1 0>, Use Rotation Vector=true) {module=InitialMeshOrientation002}
   &AddVelocity(
    Cone Angle=90
    Module.UseConeFalloff=false
    Module.UseVelDistribution=false
    Override Seed=false
    Velocity Mode=::NewEnumerator2
    Velocity Speed=20
   )
   &SkeletalMeshLocation(
    Bone / Socket Index=&RandomRangeInt(Minimum=0, Maximum=5)
    Bone Sampling Mode=::NewEnumerator4
    MaskSpawnGroup=false
    Sampled Position Offset=<0 13 0>
    Skeletal Mesh=$System.Skeletal Mesh Reader
   )
  ]
  particleUpdateScript [
   &ParticleState()
   &Color(
    Color=<rgba 0.5 0.1265 0 1>
    Scale Color=&VectorFromCurve(VectorCurve=DataInterface() {propertyValues {XCurve {keys [<at 0.7 12>,<at 1 1>]}, YCurve {keys [<at 0.7 12>,<at 1 1>]}, ZCurve {keys [<at 0.7 12>,<at 1 1>]}}})
   )
   &ScaleMeshSize(Scale Factor=&VectorFromFloat(Value=&FloatFromCurve(
      FloatCurve=DataInterface() {propertyValues {Curve {keys [{arriveTangent=2.2457, interpMode=RCIM_Cubic, leaveTangent=2.2457, tangentMode=RCTM_User, time=0, value=0}, <at 1 1>]}}}
      Scale Curve=1.2
     )))
   &SkeletalMeshLocation(
    Bone / Socket Index=&RandomRangeInt(Minimum=0, Maximum=5)
    Bone Sampling Mode=::NewEnumerator4
    MaskSpawnGroup=false
    Sampled Position Offset=<0 13 0>
    Skeletal Mesh=$System.Skeletal Mesh Reader
   ) {module=SkeletalMeshLocation001}
   /SetVariables(Particles.LightExponent=2) {isSetParameters=true}
   /SetVariables() {isSetParameters=true}
   /SetVariables(Particles.Light_Color:LinearColor=&MultiplyLinearColor(
     Linear Color=<rgba 1 0.2156 0 1>
     Scale Factor=&MakeVector4_FromFloat(Float=&FloatFromCurve(
       FloatCurve=DataInterface() {propertyValues {Curve {keys [
           {interpMode=RCIM_Cubic, tangentMode=RCTM_Break, time=0, value=1}
           {arriveTangent=-3.0264, interpMode=RCIM_Cubic, tangentMode=RCTM_Break, time=1, value=0}
          ]}}}
       Scale Curve=10
      ))
    )) {isSetParameters=true}
   &SolveForcesAndVelocity(Clamp Velocity=false, Limit Acceleration=false)
  ]
  rendererClasses [
   &NiagaraLightRendererProperties
   &NiagaraMeshRendererProperties
   &NiagaraSpriteRendererProperties
  ]
  renderers [
   {rendererClass=&NiagaraLightRendererProperties, rendererIndex=0}
   {rendererClass=&NiagaraMeshRendererProperties, rendererIndex=1}
   {rendererClass=&NiagaraSpriteRendererProperties, rendererIndex=2}
  ]
 }
 EggBurst_GlitterSparks {
  emitterUpdateScript [
   &EmitterState()
   &SpawnParticlesFromOtherEmitter(
    Attribute Reader=DataInterface() {propertyValues {EmitterBinding {bindingMode=Other, emitterName=Eggs002}}}
    Calculate Spawn Rate Per Particle=false
    Minimum Age Spawn Threshold:NiagaraFloat=&Add_Float(A=$System.Egg Burst Time, B=0)
    Spawn Rate=40
    Use Maximum Age=false
    Use Minimum Age=true
   ) {module=SpawnParticlesFromOtherEmitter001}
   &SpawnParticlesFromOtherEmitter(
    Attribute Reader=DataInterface() {propertyValues {EmitterBinding {bindingMode=Other, emitterName=Eggs002}}}
    Maximum Age Spawn Threshold:NiagaraFloat=&Add_Float(A=$System.Egg Burst Time, B=-0.25)
    Minimum Age Spawn Threshold:NiagaraFloat=&Add_Float(A=$System.Egg Burst Time, B=0)
    Spawn Group=1
    Spawn Rate=20
    Spawn Rate Per Particle Cap=5
    Use Maximum Age=false
    Use Minimum Age=true
   ) {module=SpawnParticlesFromOtherEmitter002}
  ]
  particleSpawnScript [
   /SetVariables(Particles.VisibilityTag=&SpawnGroup()) {isSetParameters=true}
   &InitializeParticle(
    Color=&SelectLinearColorFromArray(
     Array Sampling Mode=::NewEnumerator1
     Direct Array Index=&RandomRangeInt(Minimum=0, Maximum=7)
     Color Selection Array=$System.EggColors
     Random Seed=?
    )
    Lifetime=&MergeHelper_FloatSwitch(
     0=&RandomRangeFloat(Minimum=0.35, Maximum=0.7)
     1=&RandomRangeFloat(Minimum=0.3, Maximum=0.5)
     2=&RandomRangeFloat(Minimum=0.3, Maximum=0.5)
     3=&RandomRangeFloat(Minimum=1, Maximum=2.8)
     Index=&SpawnGroup()
     Num=::NewEnumerator0
    )
    Position Offset=$System.Egg 01 Burst Offset
    Sprite Size=&MergeHelper_Vector2Switch(
     0=&RandomRangeVector2D(Minimum=<4 4>, Maximum=<9 9>)
     1=&RandomRangeVector2D(Minimum=<25 25>, Maximum=<52 52>)
     2=&RandomRangeVector2D(Minimum=<25 25>, Maximum=<52 52>)
     3=&RandomRangeVector2D(Minimum=<3 3>, Maximum=<6 6>)
     4=?
     Index=&SpawnGroup()
     Num=::NewEnumerator0
    )
    Sprite Size Mode=::NewEnumerator4
    UsePositionOffset=true
    Write Material Random=true
   )
   &SampleParticlesFromOtherEmitter(
    Applied Velocity Scale:NiagaraFloat=&MergeHelper_FloatSwitch(
     0=&RandomRangeFloat(Minimum=1.5, Maximum=1)
     1=0.5
     2=?
     3=&RandomRangeFloat(Minimum=0.9, Maximum=1.8)
     Index=&SpawnGroup()
     Num=::NewEnumerator0
    )
    Reject Particle If Sampled Index Is Invalid=false
    Velocity Sampling=::NewEnumerator0
   )
   &AddVelocity(
    Cone Angle=240
    Cone Axis=<0 1 0.5>
    Inner Cone Angle=35
    Module.UseConeFalloff=false
    Module.UseVelDistribution=false
    Override Seed=false
    Velocity Mode=::NewEnumerator2
    Velocity Speed=&MergeHelper_FloatSwitch(
     0=&RandomRangeFloat(Minimum=150, Maximum=210)
     1=0
     2=?
     3=?
     Index=&SpawnGroup()
     Num=::NewEnumerator0
    )
   ) {module=AddVelocity001}
   &AddVelocity(
    Rotation Coordinate Space=::Simulation
    Velocity=&MergeHelper_VectorSwitch(
     0=<0 0 0>
     1=&RandomRangeVector(Minimum=<0 0 70>, Maximum=<0 0 45>)
     2=&RandomRangeVector(Minimum=<0 0 70>, Maximum=<0 0 45>)
     3=?
     Index=&SpawnGroup()
     Num=::NewEnumerator0
    )
   ) {module=AddVelocity002}
  ]
  particleUpdateScript [
   &GravityForce(Gravity=&MergeHelper_VectorSwitch(
     0=<0 0 -500>
     1=&VectorFromCurve(
      VectorCurve=DataInterface() {propertyValues {
        XCurve {keys [<at 0 0>,<at 1 0>]}
        YCurve {keys [<at 0 0>,<at 1 0>]}
        ZCurve {keys [
          {arriveTangent=1.0661, interpMode=RCIM_Cubic, leaveTangent=1.0661, tangentMode=RCTM_User, time=0, value=0}
          {interpMode=RCIM_Cubic, tangentMode=RCTM_User, time=1, value=1}
         ]}
       }}
      Scale Curve=&RandomRangeVector(Minimum=<0 0 -300>, Maximum=<0 0 -600>)
     )
     2=&VectorFromCurve(
      VectorCurve=DataInterface() {propertyValues {
        XCurve {keys [<at 0 0>,<at 1 0>]}
        YCurve {keys [<at 0 0>,<at 1 0>]}
        ZCurve {keys [
          {arriveTangent=1.0661, interpMode=RCIM_Cubic, leaveTangent=1.0661, tangentMode=RCTM_User, time=0, value=0}
          {interpMode=RCIM_Cubic, tangentMode=RCTM_User, time=1, value=1}
         ]}
       }}
      Scale Curve=&RandomRangeVector(Minimum=<0 0 -300>, Maximum=<0 0 -600>)
     )
     3=&RandomRangeVector(Minimum=<0 0 -110>, Maximum=<0 0 -400>)
     Index=&SpawnGroup()
     Num=::NewEnumerator0
    )) {module=GravityForce001}
   &Drag(
    Drag=&MergeHelper_FloatSwitch(
     0=&RandomRangeFloat(Minimum=7, Maximum=5)
     1=&FloatFromCurve(
      FloatCurve=DataInterface() {propertyValues {Curve {keys [
          {interpMode=RCIM_Cubic, tangentMode=RCTM_User, time=0, value=0}
          {interpMode=RCIM_Cubic, tangentMode=RCTM_User, time=0.7, value=0.9}
          {tangentMode=RCTM_User, time=1, value=0}
         ]}}}
      Scale Curve=5.5
     )
     2=&FloatFromCurve(
      FloatCurve=DataInterface() {propertyValues {Curve {keys [
          {interpMode=RCIM_Cubic, tangentMode=RCTM_User, time=0, value=0}
          {interpMode=RCIM_Cubic, tangentMode=RCTM_User, time=0.7, value=0.9}
          {tangentMode=RCTM_User, time=1, value=0}
         ]}}}
      Scale Curve=5.5
     )
     3=&MaskFloatBySpawnGroup(
      Masked Float=5.5
      Passthrough Float=&FloatFromCurve(
       FloatCurve=DataInterface() {propertyValues {Curve {keys [<at 0 0 RCIM_Cubic>, {arriveTangent=31.2634, interpMode=RCIM_Cubic, leaveTangent=31.2634, tangentMode=RCTM_User, time=0.1, value=1}]}}}
       Scale Curve=&RandomRangeFloat(Minimum=5.5, Maximum=8.5)
      )
     )
     Index=&SpawnGroup()
     Num=::NewEnumerator0
    )
    Use Linear Drag=true
    Use Rotational Drag=true
   )
   &CurlNoiseForce(
    Bias Noise=false
    Noise Frequency=&MergeHelper_FloatSwitch(0=100, 1=0, 2=?, 3=?, Index=&SpawnGroup(), Num=::NewEnumerator0)
    Noise Strength=&MergeHelper_FloatSwitch(
     0=100
     1=0
     2=?
     3=&Multiply_Float(
      A=&RandomRangeFloat(Minimum=30, Maximum=50)
      B=&Lerp_Float(A=3, B=10, Alpha=$User.CrowdCheerCurve)
     )
     Index=&SpawnGroup()
     Num=::NewEnumerator0
    )
    Pan Noise=false
   ) {module=CurlNoiseForce001}
   &AccelerationForce(Acceleration=&MergeHelper_VectorSwitch(
     0=<0 0 0>
     1=&VectorFromCurve(
      VectorCurve=DataInterface() {propertyValues {
        XCurve {keys [<at 0 0>,<at 1 0>]}
        YCurve {keys [<at 0 0>,<at 1 0>]}
        ZCurve {keys [
          {arriveTangent=-2.1026, interpMode=RCIM_Cubic, leaveTangent=-2.1026, tangentMode=RCTM_User, time=0, value=1}
          {interpMode=RCIM_Cubic, tangentMode=RCTM_User, time=1, value=0}
         ]}
       }}
      Scale Curve=&RandomRangeVector(Minimum=<0 0 0>, Maximum=<0 0 200>)
     )
     2=&VectorFromCurve(
      VectorCurve=DataInterface() {propertyValues {
        XCurve {keys [<at 0 0>,<at 1 0>]}
        YCurve {keys [<at 0 0>,<at 1 0>]}
        ZCurve {keys [
          {arriveTangent=-2.1026, interpMode=RCIM_Cubic, leaveTangent=-2.1026, tangentMode=RCTM_User, time=0, value=1}
          {interpMode=RCIM_Cubic, tangentMode=RCTM_User, time=1, value=0}
         ]}
       }}
      Scale Curve=&RandomRangeVector(Minimum=<0 0 0>, Maximum=<0 0 200>)
     )
     3=?
     Index=&SpawnGroup()
     Num=::NewEnumerator0
    ))
   &SolveForcesAndVelocity(Clamp Velocity=false, Limit Acceleration=false)
   &ParticleState()
   &ScaleSpriteSize(
    Uniform Curve Scale=&Multiply_Float(
     A=&RandomRangeFloat(Minimum=3, Maximum=7)
     B=&RandomRangeFloat(Minimum=0.5, Maximum=1, Evaluation Type=::NewEnumerator1)
    )
    Uniform Curve Sprite Scale=DataInterface() {propertyValues {Curve {keys [
        <at 0 0 RCIM_Cubic>
        {arriveTangent=-2.9089, interpMode=RCIM_Cubic, leaveTangent=-2.9089, tangentMode=RCTM_User, time=0.1, value=1}
        <at 1 0 RCIM_Cubic>
       ]}}}
   )
   &SpriteRotationRate(Rotation Rate=&RandomRangeFloat(Minimum=-5, Maximum=5))
   &ScaleColor(
    Scale RGB=&Multiply_VectorByFloat(Vector=<1 1 1>, Float=&RandomRangeFloat(Minimum=1, Maximum=4))
    ScaleA=true
    ScaleRGB=true
   ) {enabled=false}
   &ScaleColor(
    Scale Alpha=&MergeHelper_FloatSwitch(
     0=1
     1=&FloatFromCurve(
      FloatCurve=DataInterface() {propertyValues {Curve {keys [<at 0 2 RCIM_Cubic>, {tangentMode=RCTM_User, time=1, value=0}]}}}
      Scale Curve=&RandomRangeFloat(Minimum=0.8, Maximum=1)
     )
     2=&FloatFromCurve(
      FloatCurve=DataInterface() {propertyValues {Curve {keys [<at 0 2 RCIM_Cubic>, {tangentMode=RCTM_User, time=1, value=0}]}}}
      Scale Curve=&RandomRangeFloat(Minimum=0.8, Maximum=1)
     )
     3=?
     Index=&SpawnGroup()
     Num=::NewEnumerator0
    )
    Scale RGB=&MergeHelper_VectorSwitch(
     0=&Multiply_VectorByFloat(Vector=<1 1 1>, Float=&RandomRangeFloat(Minimum=1, Maximum=4))
     1=&VectorFromFloat(Value=&Multiply_Float(A=1, B=$System.ToD))
     2=&VectorFromFloat(Value=&Multiply_Float(A=1, B=$System.ToD))
     3=?
     Index=&SpawnGroup()
     Num=::NewEnumerator0
    )
    ScaleA=true
    ScaleRGB=true
   ) {module=ScaleColor001}
   &ScaleColor(
    Scale Alpha=&MergeHelper_FloatSwitch(
     0=1
     1=&FloatFromCurve(
      FloatCurve=DataInterface() {propertyValues {Curve {keys [
          {arriveTangent=-0.0888, interpMode=RCIM_Cubic, leaveTangent=-0.0888, tangentMode=RCTM_User, time=0, value=0}
          {arriveTangent=0.12, interpMode=RCIM_Cubic, leaveTangent=0.12, tangentMode=RCTM_User, time=0.15, value=0.9847}
          {arriveTangent=-2.7131, interpMode=RCIM_Cubic, leaveTangent=-2.7131, tangentMode=RCTM_User, time=1, value=0}
         ]}}}
      Scale Curve=&Multiply_Float(A=1, B=$System.ToD)
     )
     2=&FloatFromCurve(
      FloatCurve=DataInterface() {propertyValues {Curve {keys [
          {arriveTangent=-0.0888, interpMode=RCIM_Cubic, leaveTangent=-0.0888, tangentMode=RCTM_User, time=0, value=0}
          {arriveTangent=0.12, interpMode=RCIM_Cubic, leaveTangent=0.12, tangentMode=RCTM_User, time=0.15, value=0.9847}
          {arriveTangent=-2.7131, interpMode=RCIM_Cubic, leaveTangent=-2.7131, tangentMode=RCTM_User, time=1, value=0}
         ]}}}
      Scale Curve=&Multiply_Float(A=1, B=$System.ToD)
     )
     3=?
     Index=&SpawnGroup()
     Num=::NewEnumerator0
    )
    Scale RGB=&MergeHelper_VectorSwitch(
     0=&VectorFromFloat(Value=1)
     1=&VectorFromFloat(Value=&Multiply_Float(A=1, B=$System.ToD))
     2=?
     3=?
     Index=&SpawnGroup()
     Num=::NewEnumerator0
    )
    ScaleA=true
    ScaleRGB=true
   ) {module=ScaleColor002}
   /SetVariables(
    Particles.DMP00_GlitterDust:Vector4f=&MakeVector4(
     X=&Add_Float(
      A=&RandomRangeFloat(Minimum=0, Maximum=1)
      B=&FloatFromCurve(
       FloatCurve=DataInterface() {propertyValues {Curve {keys [
           {arriveTangent=2.7533, interpMode=RCIM_Cubic, leaveTangent=1.8958, tangentMode=RCTM_User, time=0, value=1}
           {arriveTangent=0.8292, interpMode=RCIM_Cubic, leaveTangent=0.8292, tangentMode=RCTM_User, time=0.321, value=1.3453}
           {arriveTangent=0.0529, interpMode=RCIM_Cubic, leaveTangent=0.2021, tangentMode=RCTM_User, time=1, value=1.5}
          ]}}}
       CurveIndex=&Add_Float(A=&Multiply_Float(A=1, B=0.5), B=$Particles.NormalizedAge)
      )
     )
     Y=0.5
     Z=0.5
     W=&FloatFromCurve(
      FloatCurve=DataInterface() {propertyValues {Curve {keys [<at 0.5 1>,<at 1 0>]}}}
      Scale Curve=&FNTimeOfDayScalar(
       Night Scale=2
       Morning Scale=2
       Day Scale=&FN_FrontEndFloatModifier(Gameplay=1, Front End=0.5)
       Evening Scale=3
      )
     )
    )
    Particles.DMP02_GlitterDust:Vector4f=&MakeVector4(
     X=&FloatFromCurve(FloatCurve=DataInterface() {propertyValues {Curve {keys [
          {interpMode=RCIM_Cubic, leaveTangent=4.2485, tangentMode=RCTM_Break, time=0, value=0}
          {arriveTangent=1.0548, interpMode=RCIM_Cubic, leaveTangent=1.0548, tangentMode=RCTM_User, time=0.17, value=0.5715}
          <at 1 1>
         ]}}})
     Y=&FloatFromCurve(
      FloatCurve=DataInterface() {propertyValues {Curve {keys [
          {arriveTangent=1.9056, interpMode=RCIM_Cubic, leaveTangent=1.9056, tangentMode=RCTM_User, time=0, value=0.2}
          {arriveTangent=0.0259, interpMode=RCIM_Cubic, leaveTangent=0.0259, tangentMode=RCTM_User, time=1, value=1.4}
         ]}}}
      Scale Curve=&RandomRangeFloat(Minimum=0.2, Maximum=0.5)
     )
     Z=&FloatFromCurve(FloatCurve=DataInterface() {propertyValues {Curve {keys [
          {interpMode=RCIM_Cubic, leaveTangent=2.3064, tangentMode=RCTM_Break, time=0, value=1}
          {interpMode=RCIM_Cubic, leaveTangent=11.1233, tangentMode=RCTM_Break, time=1, value=2}
         ]}}})
     W=&FloatFromCurve(
      FloatCurve=DataInterface() {propertyValues {Curve {keys [
          {arriveTangent=-0.6867, interpMode=RCIM_Cubic, leaveTangent=-0.6867, tangentMode=RCTM_User, time=0, value=1}
          {interpMode=RCIM_Cubic, leaveTangent=-1.2956, tangentMode=RCTM_Break, time=1, value=0.75}
         ]}}}
      Scale Curve=&RandomRangeFloat(Minimum=1.3, Maximum=1.6)
     )
    )
    Particles.GlitterDustSize:Vector2f=&Multiply_Vector2D(
     A=$Particles.Initial.SpriteSize
     B=&Vector2DFromFloat(Value=&FloatFromCurve(
       FloatCurve=DataInterface() {propertyValues {Curve {
          defaultValue=0
          keys [
           {arriveTangent=2.2927, interpMode=RCIM_Cubic, leaveTangent=2.2927, tangentMode=RCTM_User, time=0, value=0}
           {arriveTangent=0.0224, arriveTangentWeight=0.3333, interpMode=RCIM_Cubic, leaveTangent=0.0224, tangentMode=RCTM_User, time=1, value=1}
          ]
         }}}
       Scale Curve=1.2
      ))
    )
   ) {isSetParameters=true}
  ]
  rendererClasses [&NiagaraSpriteRendererProperties]
  renderers [
   {rendererClass=&NiagaraSpriteRendererProperties}
   {rendererClass=&NiagaraSpriteRendererProperties, rendererIndex=1}
  ]
 }
 Eggs002 {
  emitterUpdateScript [
   &EmitterState(
    Inactive Response=::NewEnumerator2
    Life Cycle Mode=::NewEnumerator1
    Loop Behavior=::NewEnumerator1
    Loop Delay=0.01
    Loop Duration=1.1
    UseLoopDelay=false
   )
   &SpawnRate(SpawnRate=7, Use Spawn Probability=false)
  ]
  particleSpawnScript [
   &InitializeParticle(
    Color=&SelectLinearColorFromArray(
     Array Sampling Mode=::NewEnumerator1
     Direct Array Index=&RandomRangeInt(Minimum=0, Maximum=7)
     Color Selection Array=$System.EggColors
     Random Seed=?
    )
    Lifetime=$System.Egg Burst Time
    Position Offset Coordinate Space=::Local
    Sprite Size Mode=::NewEnumerator3
    UsePositionOffset=true
    Write Material Random=true
   )
   &SkeletalMeshLocation(
    Bone / Socket Index=/CascadeToNiagaraConverter/NiagaraScripts/DynamicInputScripts/SequentialInt(Use Array Data:NiagaraBool=false, First Number:NiagaraInt32=0, Sequence:NiagaraDataInterfaceArrayInt32=?, Last Number:NiagaraInt32=5)
    Bone Sampling Mode=::NewEnumerator4
    MaskSpawnGroup=false
    Sampled Position Offset=<0 15 0>
    Skeletal Mesh=$System.Skeletal Mesh Reader
   )
   &AddVelocity(
    Module.Use Falloff Distance=false
    Origin Offset Coordinate Space=::World
    Override Seed=false
    Velocity Mode=::NewEnumerator1
    Velocity Origin=$Output.SkeletalMeshLocation.SampledPositionWithoutOffset
    Velocity Speed=&RandomRangeFloat(Minimum=450, Maximum=250)
   ) {module=AddVelocity001}
   &InheritVelocity(Inherited Velocity Speed Limit=150, Speed Limit=true, UseSpeedThreshold=false)
   &CameraOffset(Camera Offset Amount=20)
  ]
  particleUpdateScript [
   &ParticleState()
   &ScaleSpriteSize(Uniform Curve Sprite Scale=DataInterface() {propertyValues {Curve {
       defaultValue=0
       keys [
        {tangentMode=RCTM_User, time=0, value=1}
        {arriveTangent=-0.0405, interpMode=RCIM_Cubic, leaveTangent=-0.0405, tangentMode=RCTM_User, time=0.85, value=1}
        {arriveTangent=-8.0089, interpMode=RCIM_Cubic, leaveTangent=-8.0089, tangentMode=RCTM_User, time=1, value=0}
       ]
      }}})
   &ScaleColor(
    Scale RGB=&VectorFromFloat(Value=&Multiply_Float(A=1, B=$System.ToD))
    ScaleA=true
    ScaleRGB=true
   ) {module=ScaleColor003}
   &GravityForce()
   &SolveForcesAndVelocity(Clamp Velocity=false, Limit Acceleration=false)
  ]
  rendererClasses [&NiagaraSpriteRendererProperties]
  renderers [{rendererClass=&NiagaraSpriteRendererProperties}]
 }
}
summary {
 systemName=NS_Emote_HoppyBun_Eggs
 emitters [
  {emitterName=Eggs002, rendererClasses [&NiagaraSpriteRendererProperties]}
  {emitterName=Confetti001, rendererClasses [&NiagaraSpriteRendererProperties]}
  {
   emitterName=CoreFlash004
   rendererClasses [
    &NiagaraLightRendererProperties
    &NiagaraMeshRendererProperties
    &NiagaraSpriteRendererProperties
   ]
  }
  {emitterName=EggBurst_GlitterSparks, rendererClasses [&NiagaraSpriteRendererProperties]}
 ]
 userVariables [
  User.bIsFrontEnd {type=NiagaraBool}
  User.bIsFrontEndPreview {type=NiagaraBool}
  User.CrowdCheerCurve {type=NiagaraFloat, defaultValue=0, description=""}
 ]
}
